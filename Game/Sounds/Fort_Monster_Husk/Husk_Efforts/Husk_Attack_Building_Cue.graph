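# Sound cue for Husk vocal efforts, named for attacking buildings, played as a non-critical enemy sound with directional-cone attenuation and a three-voice concurrency set.
# A quality-level node picks between a full variation set and a trimmed one, with its other two slots empty.
# In each, a group control gives the first active instance a random attack vocal and overflow instances random crowd vocals, both through modulators for pitch and volume variation.
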SoundCue /Game/Sounds/Fort_Monster_Husk/Husk_Efforts/Husk_Attack_Building_Cue
root=QualityLevel_0
nodes [
 {id=GroupControl_1, children [Modulator_2,Modulator_3], properties {groupSizes [1]}}
 {id=GroupControl_2, children [Modulator_4,Modulator_5], properties {groupSizes [1]}}
 {id=Modulator_2, children [Random_2]}
 {id=Modulator_3, children [Random_3], properties {pitchMax=1.3, volumeMax=1, volumeMin=0.8}}
 {id=Modulator_4, children [Random_4]}
 {id=Modulator_5, children [Random_5], properties {pitchMax=1.3, volumeMax=1, volumeMin=0.8}}
 {id=QualityLevel_0, children [GroupControl_1, GroupControl_2, null, null]}
 {id=Random_2, children [WavePlayer_14, WavePlayer_15, WavePlayer_16, WavePlayer_17, WavePlayer_18], properties {weights [1,1,1,1,1]}}
 {
  id=Random_3
  children [WavePlayer_19, WavePlayer_20, WavePlayer_21, WavePlayer_22, WavePlayer_23, WavePlayer_24, WavePlayer_25, WavePlayer_26, WavePlayer_27]
  properties {weights [1,1,1,1,1,1,1,1,1]}
 }
 {id=Random_4, children [WavePlayer_28, WavePlayer_30, WavePlayer_31, WavePlayer_32], properties {weights [1,1,1,1]}}
 {id=Random_5, children [WavePlayer_33, WavePlayer_35, WavePlayer_36, WavePlayer_38, WavePlayer_41], properties {weights [1,1,1,1,1]}}
 {id=WavePlayer_14, properties {soundWaveAssetPtr=&/Husk_Attack_Medium01}}
 {id=WavePlayer_15, properties {soundWaveAssetPtr=&/Husk_Attack_Medium02}}
 {id=WavePlayer_16, properties {soundWaveAssetPtr=&/Husk_Attack_Medium03}}
 {id=WavePlayer_17, properties {soundWaveAssetPtr=&/Husk_Attack_Medium04}}
 {id=WavePlayer_18, properties {soundWaveAssetPtr=&/Husk_Attack_Long04}}
 {id=WavePlayer_19, properties {soundWaveAssetPtr=&/husk_newcrowd_10}}
 {id=WavePlayer_20, properties {soundWaveAssetPtr=&/husk_newcrowd_11}}
 {id=WavePlayer_21, properties {soundWaveAssetPtr=&/husk_newcrowd_12}}
 {id=WavePlayer_22, properties {soundWaveAssetPtr=&/husk_newcrowd_13}}
 {id=WavePlayer_23, properties {soundWaveAssetPtr=&/husk_newcrowd_14}}
 {id=WavePlayer_24, properties {soundWaveAssetPtr=&/husk_newcrowd_15}}
 {id=WavePlayer_25, properties {soundWaveAssetPtr=&/husk_newcrowd_16}}
 {id=WavePlayer_26, properties {soundWaveAssetPtr=&/husk_newcrowd_17}}
 {id=WavePlayer_27, properties {soundWaveAssetPtr=&/husk_newcrowd_18}}
 {id=WavePlayer_28, properties {soundWaveAssetPtr=&Husk_Attack_Medium01}}
 {id=WavePlayer_30, properties {soundWaveAssetPtr=&Husk_Attack_Medium03}}
 {id=WavePlayer_31, properties {soundWaveAssetPtr=&Husk_Attack_Medium04}}
 {id=WavePlayer_32, properties {soundWaveAssetPtr=&Husk_Attack_Long04}}
 {id=WavePlayer_33, properties {soundWaveAssetPtr=&husk_newcrowd_10}}
 {id=WavePlayer_35, properties {soundWaveAssetPtr=&husk_newcrowd_12}}
 {id=WavePlayer_36, properties {soundWaveAssetPtr=&husk_newcrowd_13}}
 {id=WavePlayer_38, properties {soundWaveAssetPtr=&husk_newcrowd_15}}
 {id=WavePlayer_41, properties {soundWaveAssetPtr=&husk_newcrowd_18}}
]
properties {
 attenuationSettings=/Game/Sounds/Attenuation/Fort_Monster_HuskDirectionalCone_Att
 maxDistance=966.4637
 soundClassObject=/Game/Sounds/FortSoundClassesAndModes/Fort_Enemy_NonCritical
 volumeMultiplier=0.6
 attenuationOverrides {distanceAlgorithm=NaturalSound, falloffDistance=4200, attenuationShapeExtents {x=800}}
 concurrencyOverrides {resolutionRule=PreventNew}
 concurrencySet [/Game/Sounds/Concurrency_Rules/ThreeMaximum_StopOldest_VolumeScale]
}
