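# Sound cue for the Husk monster's medium attack effort, played in the enemy attack sound class with directional-cone attenuation and Husk attack concurrency sets.
# A quality-level root fills two of four slots, each mixing a randomly picked attack wave with a random sweetener under separate pitch modulation.
# The first slot draws on larger pools and adds an outer modulator, while the second uses fewer waves.
# The cue's debug flag is on.
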SoundCue /Game/Sounds/Fort_Monster_Husk/Husk_Efforts/Husk_Attack_Medium_Cue
root=QualityLevel_0
nodes [
 {id=Mixer_1, children [Modulator_2,Modulator_3], properties {inputVolume [0.9,0.7]}}
 {id=Mixer_2, children [Modulator_4,Modulator_5], properties {inputVolume [0.9,1]}}
 {id=Modulator_0, children [Mixer_1], properties {pitchMax=1.2}}
 {id=Modulator_2, children [Random_2], properties {pitchMax=1.3, pitchMin=0.9}}
 {id=Modulator_3, children [Random_3], properties {pitchMax=1.3, pitchMin=0.9}}
 {id=Modulator_4, children [Random_4], properties {pitchMax=1.3, pitchMin=0.9}}
 {id=Modulator_5, children [Random_5], properties {pitchMax=1.3, pitchMin=0.9}}
 {id=QualityLevel_0, children [Modulator_0,Mixer_2,null,null]}
 {id=Random_2, children [WavePlayer_8, WavePlayer_9, WavePlayer_10, WavePlayer_16, WavePlayer_17], properties {weights [1,1,1,1,1]}}
 {id=Random_3, children [WavePlayer_18, WavePlayer_0, WavePlayer_1, WavePlayer_2, WavePlayer_3], properties {weights [1,1,1,1,1]}}
 {id=Random_4, children [WavePlayer_21, WavePlayer_24, WavePlayer_25], properties {weights [1,1,1]}}
 {id=Random_5, children [WavePlayer_26, WavePlayer_27], properties {weights [1,1]}}
 {id=WavePlayer_0, properties {soundWaveAssetPtr=&/husk_attack_sweeten_03}}
 {id=WavePlayer_1, properties {soundWaveAssetPtr=&/husk_attack_sweeten_04}}
 {id=WavePlayer_10, properties {soundWaveAssetPtr=&/Husk_Attack_Medium03}}
 {id=WavePlayer_16, properties {soundWaveAssetPtr=&/Husk_Attack_Medium04}}
 {id=WavePlayer_17, properties {soundWaveAssetPtr=&/Husk_Attack_Long04}}
 {id=WavePlayer_18, properties {soundWaveAssetPtr=&/husk_attack_sweeten_01}}
 {id=WavePlayer_2, properties {soundWaveAssetPtr=&/husk_attack_sweeten_05}}
 {id=WavePlayer_21, properties {soundWaveAssetPtr=&/Husk_Attack_Medium01}}
 {id=WavePlayer_24, properties {soundWaveAssetPtr=&Husk_Attack_Medium04}}
 {id=WavePlayer_25, properties {soundWaveAssetPtr=&Husk_Attack_Long04}}
 {id=WavePlayer_26, properties {soundWaveAssetPtr=&husk_attack_sweeten_01}}
 {id=WavePlayer_27, properties {soundWaveAssetPtr=&/husk_attack_sweeten_02}}
 {id=WavePlayer_3, properties {soundWaveAssetPtr=&/husk_attack_sweeten_06}}
 {id=WavePlayer_8, properties {soundWaveAssetPtr=&Husk_Attack_Medium01}}
 {id=WavePlayer_9, properties {soundWaveAssetPtr=&/Husk_Attack_Medium02}}
]
properties {
 attenuationSettings=/Game/Sounds/Attenuation/Fort_Monster_HuskDirectionalCone_Att
 bDebug=true
 maxDistance=966.4637
 soundClassObject=/Game/Sounds/FortSoundClassesAndModes/Fort_Enemy_Attack
 volumeMultiplier=1.5
 attenuationOverrides {distanceAlgorithm=NaturalSound, falloffDistance=4200, attenuationShapeExtents {x=800}}
 concurrencyOverrides {maxCount=5, resolutionRule=StopFarthestThenPreventNew}
 concurrencySet [/Game/Sounds/Concurrency_Rules/Husk_Attacks_Conc, /Game/Sounds/Concurrency_Rules/Husk_Enemy_Close_Attacks_Conc]
}
