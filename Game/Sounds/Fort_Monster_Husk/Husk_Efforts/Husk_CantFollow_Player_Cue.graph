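# Sound cue for Husk monster effort sounds, named for when the Husk cannot follow the player.
# A Random node picks one of four equally weighted waves and feeds a Modulator that varies each playback.
# It uses the shared Husk attenuation asset and the Fort_Enemy_Spawn sound class at reduced volume.
# A concurrency override refuses new instances once the limit is reached rather than cutting off playing ones.
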SoundCue /Game/Sounds/Fort_Monster_Husk/Husk_Efforts/Husk_CantFollow_Player_Cue
root=Modulator_20
nodes [
 {id=Modulator_20, children [Random_18]}
 {id=Random_18, children [WavePlayer_111, WavePlayer_112, WavePlayer_113, WavePlayer_114], properties {weights [1,1,1,1]}}
 {id=WavePlayer_111, properties {soundWaveAssetPtr=&/Husk_CantFollow_Player01}}
 {id=WavePlayer_112, properties {soundWaveAssetPtr=&/Husk_CantFollow_Player02}}
 {id=WavePlayer_113, properties {soundWaveAssetPtr=&/Husk_CantFollow_Player03}}
 {id=WavePlayer_114, properties {soundWaveAssetPtr=&/Husk_CantFollow_Player04}}
]
properties {
 attenuationSettings=/Game/Sounds/Attenuation/Fort_Monster_Husk_Att
 bOverrideConcurrency=true
 maxDistance=2800
 soundClassObject=/Game/Sounds/FortSoundClassesAndModes/Fort_Enemy_Spawn
 volumeMultiplier=0.6
 attenuationOverrides {distanceAlgorithm=NaturalSound, falloffDistance=4000, attenuationShapeExtents {x=1000}}
 concurrencyOverrides {resolutionRule=PreventNew}
}
