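# Sound cue for a Husk death scream, named as the daytime variant, with pain/death attenuation and two-voice concurrency.
# After a fixed short delay, a quality switch picks one of two mixers; the second draws on smaller wave pools, and two slots are empty.
# Each mixer layers a random death vocal, slightly delayed and pitch-varied, over a quieter random echo scream pitched to double.
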SoundCue /Game/Sounds/Fort_Monster_Husk/Husk_Efforts/Husk_Day_Death_Cue
root=Delay_0
nodes [
 {id=Delay_0, children [QualityLevel_0], properties {delayMax=0.2, delayMin=0.2}}
 {id=Delay_1, children [Modulator_2], properties {delayMax=0.1, delayMin=0.05}}
 {id=Delay_2, children [Modulator_4], properties {delayMax=0.1, delayMin=0.05}}
 {id=Mixer_1, children [Delay_1,Modulator_3], properties {inputVolume [1,0.4]}}
 {id=Mixer_2, children [Delay_2,Modulator_5], properties {inputVolume [1,0.4]}}
 {id=Modulator_2, children [Random_3], properties {pitchMax=1.3, pitchMin=0.9}}
 {id=Modulator_3, children [Random_2], properties {pitchMax=2, pitchMin=2}}
 {id=Modulator_4, children [Random_5], properties {pitchMax=1.3, pitchMin=0.9}}
 {id=Modulator_5, children [Random_6], properties {pitchMax=2, pitchMin=2}}
 {id=QualityLevel_0, children [Mixer_1,Mixer_2,null,null]}
 {
  id=Random_2
  children [WavePlayer_11, WavePlayer_12, WavePlayer_13, WavePlayer_14, WavePlayer_15, WavePlayer_16]
  properties {weights [1,1,1,1,1,1]}
 }
 {id=Random_3, children [WavePlayer_17, WavePlayer_18, WavePlayer_19, WavePlayer_20, WavePlayer_21], properties {weights [1,1,1,1,1]}}
 {id=Random_5, children [WavePlayer_28, WavePlayer_30, WavePlayer_31], properties {weights [1,1,1]}}
 {id=Random_6, children [WavePlayer_23, WavePlayer_26, WavePlayer_27], properties {weights [1,1,1]}}
 {id=WavePlayer_11, properties {soundWaveAssetPtr=&/Husk_DeathScream_Echo_01}}
 {id=WavePlayer_12, properties {soundWaveAssetPtr=&/Husk_DeathScream_Echo_02}}
 {id=WavePlayer_13, properties {soundWaveAssetPtr=&/Husk_DeathScream_Echo_03}}
 {id=WavePlayer_14, properties {soundWaveAssetPtr=&/Husk_DeathScream_Echo_04}}
 {id=WavePlayer_15, properties {soundWaveAssetPtr=&/Husk_DeathScream_Echo_05}}
 {id=WavePlayer_16, properties {soundWaveAssetPtr=&/Husk_DeathScream_Echo_06}}
 {id=WavePlayer_17, properties {soundWaveAssetPtr=&/husk_newdeath_A1}}
 {id=WavePlayer_18, properties {soundWaveAssetPtr=&/husk_newdeath_B1}}
 {id=WavePlayer_19, properties {soundWaveAssetPtr=&/husk_newdeath_C1}}
 {id=WavePlayer_20, properties {soundWaveAssetPtr=&/husk_newdeath_D1}}
 {id=WavePlayer_21, properties {soundWaveAssetPtr=&/husk_newdeath_E1}}
 {id=WavePlayer_23, properties {soundWaveAssetPtr=&Husk_DeathScream_Echo_02}}
 {id=WavePlayer_26, properties {soundWaveAssetPtr=&Husk_DeathScream_Echo_05}}
 {id=WavePlayer_27, properties {soundWaveAssetPtr=&Husk_DeathScream_Echo_06}}
 {id=WavePlayer_28, properties {soundWaveAssetPtr=&husk_newdeath_A1}}
 {id=WavePlayer_30, properties {soundWaveAssetPtr=&husk_newdeath_C1}}
 {id=WavePlayer_31, properties {soundWaveAssetPtr=&husk_newdeath_D1}}
]
properties {
 attenuationSettings=/Game/Sounds/Attenuation/Fort_Monster_Husk_PainDeath_Att
 maxDistance=4000
 soundClassObject=/Game/Sounds/FortSoundClassesAndModes/Fort_Enemy_Damaged_PSM
 volumeMultiplier=0.8
 concurrencyOverrides {maxCount=2, volumeScale=0.8165}
 concurrencySet [/Game/Sounds/Concurrency_Rules/TwoMaximum_StopFurthest_VolumeScale]
}
