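# Sound cue for Husk monster pain vocalizations, presumably the daytime variant, routed to the enemy-damaged sound class.
# Each mixer layers a random pain vocal over a quieter random sweetener, with randomized upward pitch shift on both.
# A quality-level root selects the full-pool mixer or a second mixer with fewer variations, leaving two slots empty.
# Husk pain/death attenuation and a two-voice stop-furthest concurrency set govern playback.
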SoundCue /Game/Sounds/Fort_Monster_Husk/Husk_Efforts/Husk_Day_Pain_Cue
root=QualityLevel_0
nodes [
 {id=Mixer_1, children [Modulator_2,Modulator_3], properties {inputVolume [1,0.6]}}
 {id=Mixer_2, children [Modulator_4,Modulator_5], properties {inputVolume [1,0.6]}}
 {id=Modulator_2, children [Random_3], properties {pitchMax=1.5, pitchMin=1}}
 {id=Modulator_3, children [Random_2], properties {pitchMax=1.5, pitchMin=1.3}}
 {id=Modulator_4, children [Random_5], properties {pitchMax=1.5, pitchMin=1}}
 {id=Modulator_5, children [Random_6], properties {pitchMax=1.5, pitchMin=1}}
 {id=QualityLevel_0, children [Mixer_1,Mixer_2,null,null]}
 {
  id=Random_2
  children [
   WavePlayer_20
   WavePlayer_21
   WavePlayer_22
   WavePlayer_23
   WavePlayer_24
   WavePlayer_25
   WavePlayer_26
   WavePlayer_27
   WavePlayer_28
   WavePlayer_29
   WavePlayer_30
   WavePlayer_31
   WavePlayer_32
   WavePlayer_33
  ]
  properties {weights [1,1,1,1,1,1,1,1,1,1,1,1,1,1]}
 }
 {
  id=Random_3
  children [WavePlayer_34, WavePlayer_35, WavePlayer_36, WavePlayer_37, WavePlayer_38, WavePlayer_39]
  properties {weights [1,1,1,1,1,1]}
 }
 {id=Random_5, children [WavePlayer_54, WavePlayer_56, WavePlayer_57, WavePlayer_59], properties {weights [1,1,1,1]}}
 {
  id=Random_6
  children [WavePlayer_40, WavePlayer_44, WavePlayer_47, WavePlayer_48, WavePlayer_49, WavePlayer_51]
  properties {weights [1,1,1,1,1,1]}
 }
 {id=WavePlayer_20, properties {soundWaveAssetPtr=&/husk_pain_sweeten_10}}
 {id=WavePlayer_21, properties {soundWaveAssetPtr=&/husk_pain_sweeten_11}}
 {id=WavePlayer_22, properties {soundWaveAssetPtr=&/husk_pain_sweeten_12}}
 {id=WavePlayer_23, properties {soundWaveAssetPtr=&/husk_pain_sweeten_13}}
 {id=WavePlayer_24, properties {soundWaveAssetPtr=&/husk_pain_sweeten_14}}
 {id=WavePlayer_25, properties {soundWaveAssetPtr=&/husk_pain_sweeten_15}}
 {id=WavePlayer_26, properties {soundWaveAssetPtr=&/husk_pain_sweeten_16}}
 {id=WavePlayer_27, properties {soundWaveAssetPtr=&/husk_pain_sweeten_01}}
 {id=WavePlayer_28, properties {soundWaveAssetPtr=&/husk_pain_sweeten_04}}
 {id=WavePlayer_29, properties {soundWaveAssetPtr=&/husk_pain_sweeten_05}}
 {id=WavePlayer_30, properties {soundWaveAssetPtr=&/husk_pain_sweeten_06}}
 {id=WavePlayer_31, properties {soundWaveAssetPtr=&/husk_pain_sweeten_07}}
 {id=WavePlayer_32, properties {soundWaveAssetPtr=&/husk_pain_sweeten_08}}
 {id=WavePlayer_33, properties {soundWaveAssetPtr=&/husk_pain_sweeten_09}}
 {id=WavePlayer_34, properties {soundWaveAssetPtr=&/husk_newpain_A}}
 {id=WavePlayer_35, properties {soundWaveAssetPtr=&/husk_newpain_B}}
 {id=WavePlayer_36, properties {soundWaveAssetPtr=&/husk_newpain_C}}
 {id=WavePlayer_37, properties {soundWaveAssetPtr=&/husk_newpain_D}}
 {id=WavePlayer_38, properties {soundWaveAssetPtr=&/husk_newpain_E}}
 {id=WavePlayer_39, properties {soundWaveAssetPtr=&/husk_newpain_F}}
 {id=WavePlayer_40, properties {soundWaveAssetPtr=&husk_pain_sweeten_10}}
 {id=WavePlayer_44, properties {soundWaveAssetPtr=&husk_pain_sweeten_14}}
 {id=WavePlayer_47, properties {soundWaveAssetPtr=&husk_pain_sweeten_01}}
 {id=WavePlayer_48, properties {soundWaveAssetPtr=&husk_pain_sweeten_04}}
 {id=WavePlayer_49, properties {soundWaveAssetPtr=&husk_pain_sweeten_05}}
 {id=WavePlayer_51, properties {soundWaveAssetPtr=&husk_pain_sweeten_07}}
 {id=WavePlayer_54, properties {soundWaveAssetPtr=&husk_newpain_A}}
 {id=WavePlayer_56, properties {soundWaveAssetPtr=&husk_newpain_C}}
 {id=WavePlayer_57, properties {soundWaveAssetPtr=&husk_newpain_D}}
 {id=WavePlayer_59, properties {soundWaveAssetPtr=&husk_newpain_F}}
]
properties {
 attenuationSettings=/Game/Sounds/Attenuation/Fort_Monster_Husk_PainDeath_Att
 maxDistance=4000
 soundClassObject=/Game/Sounds/FortSoundClassesAndModes/Fort_Enemy_Damaged_PSM
 volumeMultiplier=1
 attenuationOverrides {distanceAlgorithm=NaturalSound, falloffDistance=4000, attenuationShapeExtents {x=1000}}
 concurrencyOverrides {resolutionRule=PreventNew}
 concurrencySet [/Game/Sounds/Concurrency_Rules/TwoMaximum_StopFurthest_VolumeScale]
}
