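# Sound cue for a Husk monster falling asleep, played as an attenuated non-critical enemy sound.
# Each of two populated quality-level branches mixes a random fall-asleep effort with a louder, briefly delayed random idle vocal, both passed through modulators.
# The second quality branch has a smaller idle pool with one empty slot; the remaining two quality slots are unconnected.
# A concurrency override prevents new instances once the limit is reached.
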SoundCue /Game/Sounds/Fort_Monster_Husk/Husk_Efforts/Husk_FallAsleep_Cue
root=QualityLevel_0
nodes [
 {id=Delay_1, children [Modulator_3], properties {delayMax=0.9, delayMin=0.7}}
 {id=Delay_2, children [Modulator_5], properties {delayMax=0.9, delayMin=0.7}}
 {id=Mixer_1, children [Modulator_2,Delay_1], properties {inputVolume [0.8,1.4]}}
 {id=Mixer_2, children [Modulator_4,Delay_2], properties {inputVolume [0.8,1.4]}}
 {id=Modulator_2, children [Random_2]}
 {id=Modulator_3, children [Random_3]}
 {id=Modulator_4, children [Random_4]}
 {id=Modulator_5, children [Random_5]}
 {id=QualityLevel_0, children [Mixer_1,Mixer_2,null,null]}
 {id=Random_2, children [WavePlayer_11, WavePlayer_12, WavePlayer_13], properties {weights [1,1,1]}}
 {
  id=Random_3
  children [WavePlayer_14, WavePlayer_15, WavePlayer_16, WavePlayer_17, WavePlayer_18, WavePlayer_19, WavePlayer_20, WavePlayer_21]
  properties {weights [1,1,1,1,1,1,1,1]}
 }
 {id=Random_4, children [WavePlayer_22, WavePlayer_23, WavePlayer_24], properties {weights [1,1,1]}}
 {id=Random_5, children [WavePlayer_25, null, WavePlayer_27, WavePlayer_28, WavePlayer_29, WavePlayer_31], properties {weights [1,1,1,1,1,1]}}
 {id=WavePlayer_11, properties {soundWaveAssetPtr=&/husk_fallasleep_01}}
 {id=WavePlayer_12, properties {soundWaveAssetPtr=&/husk_fallasleep_02}}
 {id=WavePlayer_13, properties {soundWaveAssetPtr=&/husk_fallasleep_03}}
 {id=WavePlayer_14, properties {soundWaveAssetPtr=&/Husk_Idle01}}
 {id=WavePlayer_15, properties {soundWaveAssetPtr=&/Husk_Idle02}}
 {id=WavePlayer_16, properties {soundWaveAssetPtr=&/Husk_Idle03}}
 {id=WavePlayer_17, properties {soundWaveAssetPtr=&/Husk_Idle04}}
 {id=WavePlayer_18, properties {soundWaveAssetPtr=&/Husk_Idle05}}
 {id=WavePlayer_19, properties {soundWaveAssetPtr=&/Husk_Idle06}}
 {id=WavePlayer_20, properties {soundWaveAssetPtr=&/Husk_Idle07}}
 {id=WavePlayer_21, properties {soundWaveAssetPtr=&/Husk_Idle08}}
 {id=WavePlayer_22, properties {soundWaveAssetPtr=&husk_fallasleep_01}}
 {id=WavePlayer_23, properties {soundWaveAssetPtr=&husk_fallasleep_02}}
 {id=WavePlayer_24, properties {soundWaveAssetPtr=&husk_fallasleep_03}}
 {id=WavePlayer_25, properties {soundWaveAssetPtr=&Husk_Idle01}}
 {id=WavePlayer_27, properties {soundWaveAssetPtr=&Husk_Idle03}}
 {id=WavePlayer_28, properties {soundWaveAssetPtr=&Husk_Idle04}}
 {id=WavePlayer_29, properties {soundWaveAssetPtr=&Husk_Idle05}}
 {id=WavePlayer_31, properties {soundWaveAssetPtr=&Husk_Idle07}}
]
properties {
 attenuationSettings=/Game/Sounds/Attenuation/Fort_Monster_Husk_Att
 bOverrideConcurrency=true
 maxDistance=2800
 soundClassObject=/Game/Sounds/FortSoundClassesAndModes/Fort_Enemy_NonCritical
 volumeMultiplier=0.6
 concurrencyOverrides {resolutionRule=PreventNew}
}
