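# Sound cue for Husk monster laughing efforts, played as non-critical enemy audio with Husk attenuation.
# A quality-level switch selects a pool of eight laugh waves or a reduced pool of four, leaving two further slots empty.
# Each pool feeds a random pick through a modulator and a short random delay before playback.
# The cue overrides concurrency so new instances are blocked once the limit is reached.
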SoundCue /Game/Sounds/Fort_Monster_Husk/Husk_Efforts/Husk_Laughing_Cue
root=QualityLevel_0
nodes [
 {id=Delay_1, children [Modulator_1], properties {delayMax=2, delayMin=0.3}}
 {id=Delay_2, children [Modulator_2], properties {delayMax=2, delayMin=0.3}}
 {id=Modulator_1, children [Random_1]}
 {id=Modulator_2, children [Random_2]}
 {id=QualityLevel_0, children [Delay_1,Delay_2,null,null]}
 {
  id=Random_1
  children [WavePlayer_8, WavePlayer_9, WavePlayer_10, WavePlayer_11, WavePlayer_12, WavePlayer_13, WavePlayer_14, WavePlayer_15]
  properties {weights [1,1,1,1,1,1,1,1]}
 }
 {id=Random_2, children [WavePlayer_17, WavePlayer_21, WavePlayer_22, WavePlayer_23], properties {weights [1,1,1,1]}}
 {id=WavePlayer_10, properties {soundWaveAssetPtr=&/husk_laughing_03}}
 {id=WavePlayer_11, properties {soundWaveAssetPtr=&/husk_laughing_04}}
 {id=WavePlayer_12, properties {soundWaveAssetPtr=&/husk_laughing_05}}
 {id=WavePlayer_13, properties {soundWaveAssetPtr=&/husk_laughing_06}}
 {id=WavePlayer_14, properties {soundWaveAssetPtr=&/husk_laughing_07}}
 {id=WavePlayer_15, properties {soundWaveAssetPtr=&/husk_laughing_08}}
 {id=WavePlayer_17, properties {soundWaveAssetPtr=&/husk_laughing_02}}
 {id=WavePlayer_21, properties {soundWaveAssetPtr=&husk_laughing_06}}
 {id=WavePlayer_22, properties {soundWaveAssetPtr=&husk_laughing_07}}
 {id=WavePlayer_23, properties {soundWaveAssetPtr=&husk_laughing_08}}
 {id=WavePlayer_8, properties {soundWaveAssetPtr=&/husk_laughing_01}}
 {id=WavePlayer_9, properties {soundWaveAssetPtr=&husk_laughing_02}}
]
properties {
 attenuationSettings=/Game/Sounds/Attenuation/Fort_Monster_Husk_Att
 bOverrideConcurrency=true
 maxDistance=2800
 soundClassObject=/Game/Sounds/FortSoundClassesAndModes/Fort_Enemy_NonCritical
 volumeMultiplier=0.6
 concurrencyOverrides {resolutionRule=PreventNew}
}
