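# Sound cue for the vocal effort a Husk monster makes on noticing a player, in the non-critical enemy sound class with Husk attenuation.
# A quality-level node selects between two modulated random branches: seven equally weighted variants in the first, a five-variant subset in the second.
# The two remaining quality slots are empty.
# Its own concurrency override blocks new instances at the limit rather than stopping playing ones.
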SoundCue /Game/Sounds/Fort_Monster_Husk/Husk_Efforts/Husk_NoticePlayer_Cue
root=QualityLevel_0
nodes [
 {id=Modulator_1, children [Random_1]}
 {id=Modulator_2, children [Random_2]}
 {id=QualityLevel_0, children [Modulator_1,Modulator_2,null,null]}
 {
  id=Random_1
  children [WavePlayer_7, WavePlayer_8, WavePlayer_9, WavePlayer_10, WavePlayer_11, WavePlayer_12, WavePlayer_13]
  properties {weights [1,1,1,1,1,1,1]}
 }
 {id=Random_2, children [WavePlayer_15, WavePlayer_16, WavePlayer_17, WavePlayer_18, WavePlayer_19], properties {weights [1,1,1,1,1]}}
 {id=WavePlayer_10, properties {soundWaveAssetPtr=&/Husk_NoticePlayer04}}
 {id=WavePlayer_11, properties {soundWaveAssetPtr=&/Husk_NoticePlayer05}}
 {id=WavePlayer_12, properties {soundWaveAssetPtr=&/Husk_NoticePlayer06}}
 {id=WavePlayer_13, properties {soundWaveAssetPtr=&/Husk_NoticePlayer07}}
 {id=WavePlayer_15, properties {soundWaveAssetPtr=&/Husk_NoticePlayer02}}
 {id=WavePlayer_16, properties {soundWaveAssetPtr=&/Husk_NoticePlayer03}}
 {id=WavePlayer_17, properties {soundWaveAssetPtr=&Husk_NoticePlayer04}}
 {id=WavePlayer_18, properties {soundWaveAssetPtr=&Husk_NoticePlayer05}}
 {id=WavePlayer_19, properties {soundWaveAssetPtr=&Husk_NoticePlayer06}}
 {id=WavePlayer_7, properties {soundWaveAssetPtr=&/Husk_NoticePlayer01}}
 {id=WavePlayer_8, properties {soundWaveAssetPtr=&Husk_NoticePlayer02}}
 {id=WavePlayer_9, properties {soundWaveAssetPtr=&Husk_NoticePlayer03}}
]
properties {
 attenuationSettings=/Game/Sounds/Attenuation/Fort_Monster_Husk_Att
 bOverrideConcurrency=true
 maxDistance=2800
 soundClassObject=/Game/Sounds/FortSoundClassesAndModes/Fort_Enemy_NonCritical
 volumeMultiplier=1
 attenuationOverrides {distanceAlgorithm=NaturalSound, falloffDistance=4000, attenuationShapeExtents {x=1000}}
 concurrencyOverrides {resolutionRule=PreventNew}
}
