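# Sound cue for Husk running vocal efforts, mixing run breaths, idle vocals and quieter laughs as non-critical enemy audio.
# A quality switch feeds two populated branches, the second with fewer wave variations; its other two slots are empty.
# Each branch picks evenly, with replacement, among the three categories, each with its own random wave pool, and halves laugh volume through a modulator.
# Husk attenuation and group breath concurrency assets are assigned.
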SoundCue /Game/Sounds/Fort_Monster_Husk/Husk_Efforts/Husk_Run_Breath_Cue
root=QualityLevel_0
nodes [
 {id=Modulator_1, children [Random_6], properties {volumeMax=0.5, volumeMin=0.5}}
 {id=Modulator_2, children [Random_10], properties {volumeMax=0.5, volumeMin=0.5}}
 {id=QualityLevel_0, children [Random_4,Random_8,null,null]}
 {id=Random_10, children [WavePlayer_49, WavePlayer_52, WavePlayer_54, WavePlayer_55], properties {weights [1,1,1,1]}}
 {id=Random_11, children [WavePlayer_57, WavePlayer_58, WavePlayer_59, WavePlayer_60, WavePlayer_61], properties {weights [1,1,1,1,1]}}
 {id=Random_4, children [Random_7,Random_5,Modulator_1], properties {bRandomizeWithoutReplacement=false, weights [1,1,1]}}
 {
  id=Random_5
  children [WavePlayer_21, WavePlayer_22, WavePlayer_23, WavePlayer_24, WavePlayer_25, WavePlayer_26, WavePlayer_27]
  properties {weights [1,1,1,1,1,1,1]}
 }
 {
  id=Random_6
  children [WavePlayer_28, WavePlayer_29, WavePlayer_30, WavePlayer_31, WavePlayer_32, WavePlayer_33, WavePlayer_34, WavePlayer_35]
  properties {weights [1,1,1,1,1,1,1,1]}
 }
 {
  id=Random_7
  children [WavePlayer_36, WavePlayer_37, WavePlayer_38, WavePlayer_39, WavePlayer_40, WavePlayer_41]
  properties {weights [1,1,1,1,1,1]}
 }
 {id=Random_8, children [Random_11,Random_9,Modulator_2], properties {bRandomizeWithoutReplacement=false, weights [1,1,1]}}
 {id=Random_9, children [WavePlayer_42, WavePlayer_44, WavePlayer_46, WavePlayer_47, WavePlayer_48], properties {weights [1,1,1,1,1]}}
 {id=WavePlayer_21, properties {soundWaveAssetPtr=&/Husk_Idle_ST_01}}
 {id=WavePlayer_22, properties {soundWaveAssetPtr=&/Husk_Idle_ST_02}}
 {id=WavePlayer_23, properties {soundWaveAssetPtr=&/Husk_Idle_ST_03}}
 {id=WavePlayer_24, properties {soundWaveAssetPtr=&/Husk_Idle_ST_04}}
 {id=WavePlayer_25, properties {soundWaveAssetPtr=&/Husk_Idle_ST_05}}
 {id=WavePlayer_26, properties {soundWaveAssetPtr=&/Husk_Idle_ST_06}}
 {id=WavePlayer_27, properties {soundWaveAssetPtr=&/Husk_Idle_ST_07}}
 {id=WavePlayer_28, properties {soundWaveAssetPtr=&/husk_laughing_ST02}}
 {id=WavePlayer_29, properties {soundWaveAssetPtr=&/husk_laughing_ST_01}}
 {id=WavePlayer_30, properties {soundWaveAssetPtr=&/husk_laughing_ST_03}}
 {id=WavePlayer_31, properties {soundWaveAssetPtr=&/husk_laughing_ST_04}}
 {id=WavePlayer_32, properties {soundWaveAssetPtr=&/husk_laughing_ST_05}}
 {id=WavePlayer_33, properties {soundWaveAssetPtr=&/husk_laughing_ST_06}}
 {id=WavePlayer_34, properties {soundWaveAssetPtr=&/husk_laughing_ST_07}}
 {id=WavePlayer_35, properties {soundWaveAssetPtr=&/husk_laughing_ST_08}}
 {id=WavePlayer_36, properties {soundWaveAssetPtr=&/Husk_Run_Breath_1}}
 {id=WavePlayer_37, properties {soundWaveAssetPtr=&/Husk_Run_Breath_3}}
 {id=WavePlayer_38, properties {soundWaveAssetPtr=&/Husk_Run_Breath_4}}
 {id=WavePlayer_39, properties {soundWaveAssetPtr=&/Husk_Run_Breath_6}}
 {id=WavePlayer_40, properties {soundWaveAssetPtr=&/Husk_Run_Breath_7}}
 {id=WavePlayer_41, properties {soundWaveAssetPtr=&/Husk_Run_Breath_8}}
 {id=WavePlayer_42, properties {soundWaveAssetPtr=&Husk_Idle_ST_01}}
 {id=WavePlayer_44, properties {soundWaveAssetPtr=&Husk_Idle_ST_03}}
 {id=WavePlayer_46, properties {soundWaveAssetPtr=&Husk_Idle_ST_05}}
 {id=WavePlayer_47, properties {soundWaveAssetPtr=&Husk_Idle_ST_06}}
 {id=WavePlayer_48, properties {soundWaveAssetPtr=&Husk_Idle_ST_07}}
 {id=WavePlayer_49, properties {soundWaveAssetPtr=&husk_laughing_ST02}}
 {id=WavePlayer_52, properties {soundWaveAssetPtr=&husk_laughing_ST_04}}
 {id=WavePlayer_54, properties {soundWaveAssetPtr=&husk_laughing_ST_06}}
 {id=WavePlayer_55, properties {soundWaveAssetPtr=&husk_laughing_ST_07}}
 {id=WavePlayer_57, properties {soundWaveAssetPtr=&Husk_Run_Breath_1}}
 {id=WavePlayer_58, properties {soundWaveAssetPtr=&Husk_Run_Breath_3}}
 {id=WavePlayer_59, properties {soundWaveAssetPtr=&Husk_Run_Breath_4}}
 {id=WavePlayer_60, properties {soundWaveAssetPtr=&Husk_Run_Breath_6}}
 {id=WavePlayer_61, properties {soundWaveAssetPtr=&Husk_Run_Breath_7}}
]
properties {
 attenuationSettings=/Game/Sounds/Attenuation/Fort_Monster_Husk_Att
 maxDistance=2800
 soundClassObject=/Game/Sounds/FortSoundClassesAndModes/Fort_Enemy_NonCritical
 volumeMultiplier=0.85
 attenuationOverrides {distanceAlgorithm=NaturalSound, falloffDistance=4000, attenuationShapeExtents {x=1000}}
 concurrencyOverrides {maxCount=4}
 concurrencySet [/Game/Sounds/Concurrency_Rules/Husk_Enemy_Group_Breath_Conc]
}
