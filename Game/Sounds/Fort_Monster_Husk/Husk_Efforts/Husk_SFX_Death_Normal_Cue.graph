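# Sound cue for the Husk monster's normal death effect, playing a randomly chosen death zap wave.
# A quality-level switch feeds two equally weighted random pools, six waves in the first slot and a three-wave subset in the second, leaving two slots empty.
# Each pool runs through a modulator that randomly raises pitch.
# The cue uses pain/death attenuation, an enemy-damage sound class, reduced volume and a tight concurrency cap.
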SoundCue /Game/Sounds/Fort_Monster_Husk/Husk_Efforts/Husk_SFX_Death_Normal_Cue
root=QualityLevel_1
nodes [
 {id=Modulator_1, children [Random_1], properties {pitchMax=1.4, pitchMin=1}}
 {id=Modulator_2, children [Random_2], properties {pitchMax=1.4, pitchMin=1}}
 {id=QualityLevel_1, children [Modulator_1,Modulator_2,null,null]}
 {
  id=Random_1
  children [WavePlayer_6, WavePlayer_7, WavePlayer_8, WavePlayer_9, WavePlayer_10, WavePlayer_11]
  properties {weights [1,1,1,1,1,1]}
 }
 {id=Random_2, children [WavePlayer_12, WavePlayer_13, WavePlayer_17], properties {weights [1,1,1]}}
 {id=WavePlayer_10, properties {soundWaveAssetPtr=/Game/Sounds/Fort_Monster_Husk/SFX/husk_death_zap_D_05}}
 {id=WavePlayer_11, properties {soundWaveAssetPtr=/Game/Sounds/Fort_Monster_Husk/SFX/husk_death_zap_D_06}}
 {id=WavePlayer_12, properties {soundWaveAssetPtr=/Game/Sounds/Fort_Monster_Husk/SFX/husk_death_zap_D_01}}
 {id=WavePlayer_13, properties {soundWaveAssetPtr=/Game/Sounds/Fort_Monster_Husk/SFX/husk_death_zap_D_02}}
 {id=WavePlayer_17, properties {soundWaveAssetPtr=&husk_death_zap_D_06}}
 {id=WavePlayer_6, properties {soundWaveAssetPtr=&husk_death_zap_D_01}}
 {id=WavePlayer_7, properties {soundWaveAssetPtr=&husk_death_zap_D_02}}
 {id=WavePlayer_8, properties {soundWaveAssetPtr=/Game/Sounds/Fort_Monster_Husk/SFX/husk_death_zap_D_03}}
 {id=WavePlayer_9, properties {soundWaveAssetPtr=/Game/Sounds/Fort_Monster_Husk/SFX/husk_death_zap_D_04}}
]
properties {
 attenuationSettings=/Game/Sounds/Attenuation/Fort_Monster_Husk_PainDeath_Att
 bOverrideConcurrency=true
 maxDistance=4000
 soundClassObject=/Game/Sounds/FortSoundClassesAndModes/Fort_Enemy_Damaged_PSM
 volumeMultiplier=0.35
 concurrencyOverrides {maxCount=2}
}
