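# Sound cue for sleeping Husk breathing, played as a non-critical enemy sound with sleeping-Husk attenuation.
# A weighted random node picks breath clips twice as often as the moan group, which also contains one breath clip.
# Each group randomly selects one wave, and a root modulator varies its pitch.
# Concurrency allows one instance per owner and blocks new plays while it sounds.
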SoundCue /Game/Sounds/Fort_Monster_Husk/Husk_Efforts/Husk_Sleep_Breath_Cue
root=Modulator_7
nodes [
 {id=Modulator_7, children [Random_18], properties {pitchMax=1.2, pitchMin=0.9}}
 {
  id=Random_16
  children [
   WavePlayer_102
   WavePlayer_103
   WavePlayer_104
   WavePlayer_105
   WavePlayer_106
   WavePlayer_107
   WavePlayer_108
   WavePlayer_109
   WavePlayer_110
  ]
  properties {weights [1,1,1,1,1,1,1,1,1]}
 }
 {
  id=Random_17
  children [
   WavePlayer_111
   WavePlayer_112
   WavePlayer_113
   WavePlayer_114
   WavePlayer_115
   WavePlayer_116
   WavePlayer_117
   WavePlayer_118
   WavePlayer_119
   WavePlayer_120
   WavePlayer_121
  ]
  properties {weights [1,1,1,1,1,1,1,1,1,1,1]}
 }
 {id=Random_18, children [Random_16,Random_17], properties {bRandomizeWithoutReplacement=false, weights [1,0.5]}}
 {id=WavePlayer_102, properties {soundWaveAssetPtr=&/husk_sleep_loop_b_03}}
 {id=WavePlayer_103, properties {soundWaveAssetPtr=&/husk_sleep_loop_b_07}}
 {id=WavePlayer_104, properties {soundWaveAssetPtr=&/husk_sleep_loop_b_10}}
 {id=WavePlayer_105, properties {soundWaveAssetPtr=&/husk_sleep_loop_b_01}}
 {id=WavePlayer_106, properties {soundWaveAssetPtr=&/husk_sleep_loop_b_02}}
 {id=WavePlayer_107, properties {soundWaveAssetPtr=&/husk_sleep_loop_b_04}}
 {id=WavePlayer_108, properties {soundWaveAssetPtr=&/husk_sleep_loop_b_05}}
 {id=WavePlayer_109, properties {soundWaveAssetPtr=&/husk_sleep_loop_b_06}}
 {id=WavePlayer_110, properties {soundWaveAssetPtr=&/husk_sleep_loop_b_08}}
 {id=WavePlayer_111, properties {soundWaveAssetPtr=&husk_sleep_loop_b_06}}
 {id=WavePlayer_112, properties {soundWaveAssetPtr=&/husk_sleep_moan_10}}
 {id=WavePlayer_113, properties {soundWaveAssetPtr=&/husk_sleep_moan_01}}
 {id=WavePlayer_114, properties {soundWaveAssetPtr=&/husk_sleep_moan_02}}
 {id=WavePlayer_115, properties {soundWaveAssetPtr=&/husk_sleep_moan_03}}
 {id=WavePlayer_116, properties {soundWaveAssetPtr=&/husk_sleep_moan_04}}
 {id=WavePlayer_117, properties {soundWaveAssetPtr=&/husk_sleep_moan_05}}
 {id=WavePlayer_118, properties {soundWaveAssetPtr=&/husk_sleep_moan_06}}
 {id=WavePlayer_119, properties {soundWaveAssetPtr=&/husk_sleep_moan_07}}
 {id=WavePlayer_120, properties {soundWaveAssetPtr=&/husk_sleep_moan_08}}
 {id=WavePlayer_121, properties {soundWaveAssetPtr=&/husk_sleep_moan_09}}
]
properties {
 attenuationSettings=/Game/Sounds/Attenuation/Fort_Monster_Husk_Sleeping_Att
 bOverrideConcurrency=true
 maxDistance=2100
 soundClassObject=/Game/Sounds/FortSoundClassesAndModes/Fort_Enemy_NonCritical
 volumeMultiplier=1.5
 attenuationOverrides {distanceAlgorithm=Logarithmic, falloffDistance=3200}
 concurrencyOverrides {bLimitToOwner=true, maxCount=1, resolutionRule=PreventNew}
}
