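# Sound cue for Husk monster stunned vocal efforts, using Husk attenuation and the enemy-damaged sound class.
# A quality switch selects between two filled branches, the first with seven individual stun variations and the second with four.
# In each, a group control limits concurrent pitched-down random individual stuns to one, sending further instances to random crowd stun waves.
# Concurrency override blocks new sounds at the limit.
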SoundCue /Game/Sounds/Fort_Monster_Husk/Husk_Efforts/Husk_Stunned_Cue
root=QualityLevel_0
nodes [
 {id=Attenuation_1, children [GroupControl_1], properties {bOverrideAttenuation=true}}
 {id=Attenuation_2, children [GroupControl_2], properties {bOverrideAttenuation=true}}
 {id=GroupControl_1, children [Modulator_1,Random_3], properties {groupSizes [1]}}
 {id=GroupControl_2, children [Modulator_2,Random_5], properties {groupSizes [1]}}
 {id=Modulator_1, children [Random_2], properties {pitchMax=0.8, pitchMin=0.6}}
 {id=Modulator_2, children [Random_4], properties {pitchMax=0.8, pitchMin=0.6}}
 {id=QualityLevel_0, children [Attenuation_1, Attenuation_2, null, null]}
 {
  id=Random_2
  children [WavePlayer_9, WavePlayer_10, WavePlayer_11, WavePlayer_12, WavePlayer_13, WavePlayer_14, WavePlayer_15]
  properties {weights [1,1,1,1,1,1,1]}
 }
 {id=Random_3, children [WavePlayer_16, WavePlayer_17], properties {weights [1,1]}}
 {id=Random_4, children [WavePlayer_18, WavePlayer_19, WavePlayer_21, WavePlayer_22], properties {weights [1,1,1,1]}}
 {id=Random_5, children [WavePlayer_25, WavePlayer_26], properties {weights [1,1]}}
 {id=WavePlayer_10, properties {soundWaveAssetPtr=&/husk_newstun_06}}
 {id=WavePlayer_11, properties {soundWaveAssetPtr=&/husk_newstun_05}}
 {id=WavePlayer_12, properties {soundWaveAssetPtr=&/husk_newstun_04}}
 {id=WavePlayer_13, properties {soundWaveAssetPtr=&/husk_newstun_03}}
 {id=WavePlayer_14, properties {soundWaveAssetPtr=&/husk_newstun_02}}
 {id=WavePlayer_15, properties {soundWaveAssetPtr=&/husk_newstun_01}}
 {id=WavePlayer_16, properties {soundWaveAssetPtr=&/husk_crowd_stun_01A}}
 {id=WavePlayer_17, properties {soundWaveAssetPtr=&/husk_crowd_stun_02A}}
 {id=WavePlayer_18, properties {soundWaveAssetPtr=&/husk_newstun_07}}
 {id=WavePlayer_19, properties {soundWaveAssetPtr=&husk_newstun_06}}
 {id=WavePlayer_21, properties {soundWaveAssetPtr=&husk_newstun_04}}
 {id=WavePlayer_22, properties {soundWaveAssetPtr=&husk_newstun_03}}
 {id=WavePlayer_25, properties {soundWaveAssetPtr=&husk_crowd_stun_01A}}
 {id=WavePlayer_26, properties {soundWaveAssetPtr=&husk_crowd_stun_02A}}
 {id=WavePlayer_9, properties {soundWaveAssetPtr=&husk_newstun_07}}
]
properties {
 attenuationSettings=/Game/Sounds/Attenuation/Fort_Monster_Husk_Att
 bOverrideConcurrency=true
 maxDistance=4000
 soundClassObject=/Game/Sounds/FortSoundClassesAndModes/Fort_Enemy_Damaged_PSM
 volumeMultiplier=1
 attenuationOverrides {distanceAlgorithm=NaturalSound, falloffDistance=4000, attenuationShapeExtents {x=1000}}
 concurrencyOverrides {resolutionRule=PreventNew}
}
