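# Sound cue for the Husk monster's stunned idle effort sounds, routed to the enemy-damaged sound class.
# A quality-level switch feeds two branches, each randomly picking an equally weighted wave and passing it through a modulator and overriding attenuation node.
# The first branch has four variations, the second drops to three, and the remaining two quality slots are empty.
# The cue references shared Husk attenuation and overrides concurrency to prevent new instances at the limit.
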SoundCue /Game/Sounds/Fort_Monster_Husk/Husk_Efforts/Husk_Stunned_Idle_Cue
root=QualityLevel_0
nodes [
 {id=Attenuation_1, children [Modulator_1], properties {bOverrideAttenuation=true}}
 {id=Attenuation_2, children [Modulator_2], properties {bOverrideAttenuation=true}}
 {id=Modulator_1, children [Random_1]}
 {id=Modulator_2, children [Random_2]}
 {id=QualityLevel_0, children [Attenuation_1, Attenuation_2, null, null]}
 {id=Random_1, children [WavePlayer_7,WavePlayer_6,WavePlayer_5,WavePlayer_8], properties {weights [1,1,1,1]}}
 {id=Random_2, children [WavePlayer_12, WavePlayer_11, WavePlayer_10], properties {weights [1,1,1]}}
 {id=WavePlayer_10, properties {soundWaveAssetPtr=&/Husk_Stunned01}}
 {id=WavePlayer_11, properties {soundWaveAssetPtr=&/Husk_Stunned02}}
 {id=WavePlayer_12, properties {soundWaveAssetPtr=&/Husk_Stunned04}}
 {id=WavePlayer_5, properties {soundWaveAssetPtr=&Husk_Stunned01}}
 {id=WavePlayer_6, properties {soundWaveAssetPtr=&Husk_Stunned02}}
 {id=WavePlayer_7, properties {soundWaveAssetPtr=&Husk_Stunned04}}
 {id=WavePlayer_8, properties {soundWaveAssetPtr=&/Husk_Stunned05}}
]
properties {
 attenuationSettings=/Game/Sounds/Attenuation/Fort_Monster_Husk_Att
 bOverrideConcurrency=true
 maxDistance=4000
 soundClassObject=/Game/Sounds/FortSoundClassesAndModes/Fort_Enemy_Damaged_PSM
 volumeMultiplier=0.6
 attenuationOverrides {distanceAlgorithm=NaturalSound, falloffDistance=4000, attenuationShapeExtents {x=1000}}
 concurrencyOverrides {resolutionRule=PreventNew}
}
