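# Sound cue for the ranged Husk's throw wind-up effort, classed as an enemy attack with directional-cone attenuation and a concurrency override that prevents new instances at the limit.
# Each mix layers a random, modulated throw vocal, a quieter random bone-mess effect, and a random WakeSweet clip after a fixed delay.
# A quality-level switch selects between two such mixes, the second trimming the throw and bone variation pools to fewer waves.
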SoundCue /Game/Sounds/Fort_Monster_Husk/Husk_Efforts/Husk_ThrowWindup_Cue
root=QualityLevel_0
nodes [
 {id=Delay_1, children [Modulator_3], properties {delayMax=0.65, delayMin=0.65}}
 {id=Delay_2, children [Modulator_5], properties {delayMax=0.65, delayMin=0.65}}
 {id=Mixer_1, children [Modulator_2,Delay_1,Random_5], properties {inputVolume [1,1,0.7]}}
 {id=Mixer_2, children [Modulator_4,Delay_2,Random_8], properties {inputVolume [1,1,0.7]}}
 {id=Modulator_2, children [Random_4]}
 {id=Modulator_3, children [Random_3], properties {pitchMax=1, pitchMin=1}}
 {id=Modulator_4, children [Random_7]}
 {id=Modulator_5, children [Random_6], properties {pitchMax=1, pitchMin=1}}
 {id=QualityLevel_0, children [Mixer_1,Mixer_2,null,null]}
 {id=Random_3, children [WavePlayer_16, WavePlayer_17], properties {weights [1,1]}}
 {
  id=Random_4
  children [WavePlayer_18, WavePlayer_19, WavePlayer_20, WavePlayer_21, WavePlayer_22, WavePlayer_23, WavePlayer_24]
  properties {weights [1,1,1,1,1,1,1]}
 }
 {
  id=Random_5
  children [WavePlayer_25, WavePlayer_26, WavePlayer_27, WavePlayer_28, WavePlayer_29, WavePlayer_30, WavePlayer_31]
  properties {weights [1,1,1,1,1,1,1]}
 }
 {id=Random_6, children [WavePlayer_32, WavePlayer_33], properties {weights [1,1]}}
 {id=Random_7, children [WavePlayer_34, WavePlayer_35, WavePlayer_37, WavePlayer_40], properties {weights [1,1,1,1]}}
 {id=Random_8, children [WavePlayer_41, WavePlayer_42, WavePlayer_45, WavePlayer_46], properties {weights [1,1,1,1]}}
 {id=WavePlayer_16, properties {soundWaveAssetPtr=&/Husk_WakeSweet_01}}
 {id=WavePlayer_17, properties {soundWaveAssetPtr=&/Husk_WakeSweet_05}}
 {id=WavePlayer_18, properties {soundWaveAssetPtr=&/ranged_husk_throw_new_01}}
 {id=WavePlayer_19, properties {soundWaveAssetPtr=&/ranged_husk_throw_new_02}}
 {id=WavePlayer_20, properties {soundWaveAssetPtr=&/ranged_husk_throw_new_03}}
 {id=WavePlayer_21, properties {soundWaveAssetPtr=&/ranged_husk_throw_new_04}}
 {id=WavePlayer_22, properties {soundWaveAssetPtr=&/ranged_husk_throw_new_05}}
 {id=WavePlayer_23, properties {soundWaveAssetPtr=&/ranged_husk_throw_new_06}}
 {id=WavePlayer_24, properties {soundWaveAssetPtr=&/ranged_husk_throw_new_07}}
 {id=WavePlayer_25, properties {soundWaveAssetPtr=/Game/Sounds/Fort_Monster_Husk/SFX/bone_mess_01}}
 {id=WavePlayer_26, properties {soundWaveAssetPtr=/Game/Sounds/Fort_Monster_Husk/SFX/bone_mess_07}}
 {id=WavePlayer_27, properties {soundWaveAssetPtr=/Game/Sounds/Fort_Monster_Husk/SFX/bone_mess_02}}
 {id=WavePlayer_28, properties {soundWaveAssetPtr=/Game/Sounds/Fort_Monster_Husk/SFX/bone_mess_03}}
 {id=WavePlayer_29, properties {soundWaveAssetPtr=/Game/Sounds/Fort_Monster_Husk/SFX/bone_mess_04}}
 {id=WavePlayer_30, properties {soundWaveAssetPtr=/Game/Sounds/Fort_Monster_Husk/SFX/bone_mess_05}}
 {id=WavePlayer_31, properties {soundWaveAssetPtr=/Game/Sounds/Fort_Monster_Husk/SFX/bone_mess_08}}
 {id=WavePlayer_32, properties {soundWaveAssetPtr=&Husk_WakeSweet_01}}
 {id=WavePlayer_33, properties {soundWaveAssetPtr=&Husk_WakeSweet_05}}
 {id=WavePlayer_34, properties {soundWaveAssetPtr=&ranged_husk_throw_new_01}}
 {id=WavePlayer_35, properties {soundWaveAssetPtr=&ranged_husk_throw_new_02}}
 {id=WavePlayer_37, properties {soundWaveAssetPtr=&ranged_husk_throw_new_04}}
 {id=WavePlayer_40, properties {soundWaveAssetPtr=&ranged_husk_throw_new_07}}
 {id=WavePlayer_41, properties {soundWaveAssetPtr=&bone_mess_01}}
 {id=WavePlayer_42, properties {soundWaveAssetPtr=&bone_mess_07}}
 {id=WavePlayer_45, properties {soundWaveAssetPtr=&bone_mess_04}}
 {id=WavePlayer_46, properties {soundWaveAssetPtr=&bone_mess_05}}
]
properties {
 attenuationSettings=/Game/Sounds/Attenuation/Fort_Monster_RangedHuskDirectionalCone_Att
 bOverrideConcurrency=true
 maxDistance=4940.1494
 soundClassObject=/Game/Sounds/FortSoundClassesAndModes/Fort_Enemy_Attack
 attenuationOverrides {distanceAlgorithm=NaturalSound, falloffDistance=4000, attenuationShapeExtents {x=1000}}
 concurrencyOverrides {resolutionRule=PreventNew}
}
