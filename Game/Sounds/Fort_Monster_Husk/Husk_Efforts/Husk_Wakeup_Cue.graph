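# Sound cue for Husk monster wakeup vocal efforts, classed as non-critical enemy audio with Husk sleeping attenuation.
# After a short fixed delay, a quality switch selects a mixer that layers a random Wakeup wave with a louder random WakeSweet wave, each through a modulator.
# The second quality branch uses smaller wave pools, and the last two are empty.
# Overridden concurrency caps simultaneous instances, stopping the farthest first.
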SoundCue /Game/Sounds/Fort_Monster_Husk/Husk_Efforts/Husk_Wakeup_Cue
root=Delay_0
nodes [
 {id=Delay_0, children [QualityLevel_0], properties {delayMax=0.2, delayMin=0.2}}
 {id=Mixer_1, children [Modulator_2,Modulator_3], properties {inputVolume [1,1.6]}}
 {id=Mixer_2, children [Modulator_4,Modulator_5], properties {inputVolume [1,1.6]}}
 {id=Modulator_2, children [Random_2]}
 {id=Modulator_3, children [Random_3]}
 {id=Modulator_4, children [Random_4]}
 {id=Modulator_5, children [Random_5]}
 {id=QualityLevel_0, children [Mixer_1,Mixer_2,null,null]}
 {
  id=Random_2
  children [WavePlayer_11, WavePlayer_12, WavePlayer_13, WavePlayer_14, WavePlayer_15, WavePlayer_16]
  properties {weights [1,1,1,1,1,1]}
 }
 {id=Random_3, children [WavePlayer_17, WavePlayer_18, WavePlayer_19, WavePlayer_20, WavePlayer_21], properties {weights [1,1,1,1,1]}}
 {id=Random_4, children [WavePlayer_22, WavePlayer_23, WavePlayer_24, WavePlayer_25], properties {weights [1,1,1,1]}}
 {id=Random_5, children [WavePlayer_28, WavePlayer_29, WavePlayer_30], properties {weights [1,1,1]}}
 {id=WavePlayer_11, properties {soundWaveAssetPtr=&/Husk_Wakeup08}}
 {id=WavePlayer_12, properties {soundWaveAssetPtr=&/Husk_Wakeup09}}
 {id=WavePlayer_13, properties {soundWaveAssetPtr=&/Husk_Wakeup04}}
 {id=WavePlayer_14, properties {soundWaveAssetPtr=&/Husk_Wakeup05}}
 {id=WavePlayer_15, properties {soundWaveAssetPtr=&/Husk_Wakeup06}}
 {id=WavePlayer_16, properties {soundWaveAssetPtr=&/Husk_Wakeup07}}
 {id=WavePlayer_17, properties {soundWaveAssetPtr=&/Husk_WakeSweet_04}}
 {id=WavePlayer_18, properties {soundWaveAssetPtr=&/Husk_WakeSweet_01}}
 {id=WavePlayer_19, properties {soundWaveAssetPtr=&/Husk_WakeSweet_02}}
 {id=WavePlayer_20, properties {soundWaveAssetPtr=&/Husk_WakeSweet_03}}
 {id=WavePlayer_21, properties {soundWaveAssetPtr=&/Husk_WakeSweet_05}}
 {id=WavePlayer_22, properties {soundWaveAssetPtr=&Husk_Wakeup08}}
 {id=WavePlayer_23, properties {soundWaveAssetPtr=&Husk_Wakeup09}}
 {id=WavePlayer_24, properties {soundWaveAssetPtr=&Husk_Wakeup04}}
 {id=WavePlayer_25, properties {soundWaveAssetPtr=&Husk_Wakeup05}}
 {id=WavePlayer_28, properties {soundWaveAssetPtr=&Husk_WakeSweet_04}}
 {id=WavePlayer_29, properties {soundWaveAssetPtr=&Husk_WakeSweet_01}}
 {id=WavePlayer_30, properties {soundWaveAssetPtr=&Husk_WakeSweet_02}}
]
properties {
 attenuationSettings=/Game/Sounds/Attenuation/Fort_Monster_Husk_Sleeping_Att
 bOverrideConcurrency=true
 maxDistance=2100
 soundClassObject=/Game/Sounds/FortSoundClassesAndModes/Fort_Enemy_NonCritical
 volumeMultiplier=1.2
 attenuationOverrides {distanceAlgorithm=NaturalSound, falloffDistance=4000, attenuationShapeExtents {x=1000}}
 concurrencyOverrides {maxCount=3, resolutionRule=StopFarthestThenPreventNew}
}
